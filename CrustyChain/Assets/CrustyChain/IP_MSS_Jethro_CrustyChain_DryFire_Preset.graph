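# MetaSound Source preset that supplies the first-person dry-fire sound for the Jethro CrustyChain weapon as a stereo one-shot.
# It inherits from the shared FPS weapon reload MetaSound and supplies a single dry-fire wave.
# Inputs set output volume, pan amount and narrow random ranges for pitch shift and volume.
# Playback uses first-person foley attenuation with the weapon foley sound class and concurrency set.
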MetaSoundSource /CrustyChain/Assets/CrustyChain/IP_MSS_Jethro_CrustyChain_DryFire_Preset
preset_parent=/FortMetaSoundWeapon/Items/Metasounds/FPS/MSS_FPS_Weapon_Reload
inputs [
 OutputVolume {type=Float}
 PanAmount {type=Float}
 Random.PitchShift.Max {type=Float}
 Random.PitchShift.Min {type=Float}
 Random.Volume.Max {type=Float}
 Random.Volume.Min {type=Float}
 Sounds {type=WaveAsset:Array}
 UE.Source.OnPlay {type=Trigger}
]
interfaces [UE.OutputFormat.Stereo:1.0, UE.Source.OneShot:1.0, UE.Source:1.1]
outputs [
 "UE.OutputFormat.Stereo.Audio:0" {type=Audio}
 "UE.OutputFormat.Stereo.Audio:1" {type=Audio}
 UE.Source.OneShot.OnFinished {type=Trigger}
]
pages [Default {input_defaults {
   OutputVolume=1
   PanAmount=0.3
   Random.PitchShift.Max=1.05
   Random.PitchShift.Min=0.95
   Random.Volume.Max=1.05
   Random.Volume.Min=0.95
   UE.Source.OnPlay=null
   Sounds [&/IP_Jethro_CrustyChain_DryFire_01]
  }}]
properties {
 assetImportData=&self:AssetImportData
 attenuationSettings=/FortMetaSoundWeapon/Items/Shared/Settings/FPS/AS_Weapon_FP_Foley
 maxDistance=2250
 numChannels=2
 outputFormat=Stereo
 soundClassObject=&Fort_WeaponFoley
 concurrencySet [&CS_Fort_WeaponFoley]
}
