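# One-shot mono MetaSound for the CrustyChain weapon's overheat-complete sound, routed through weapon foley sound class and attenuation.
# On play, a shuffle array picks one of three wave variations and triggers a mono wave player with it.
# The player's output is scaled by the MasterVolume input, and the source finishes when the wave ends.
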MetaSoundSource /CrustyChain/Assets/CrustyChain/IP_MSS_Jethro_CrustyChain_Overheat_Complete
inputs [
 MasterVolume {type=Float}
 Sounds {type=WaveAsset:Array}
 UE.Source.OnPlay {type=Trigger}
]
interfaces [UE.OutputFormat.Mono:1.0, UE.Source.OneShot:1.0, UE.Source:1.1]
outputs [
 "UE.OutputFormat.Mono.Audio:0" {type=Audio}
 UE.Source.OneShot.OnFinished {type=Trigger}
]
pages [Default {
  input_defaults {
   MasterVolume=0.9
   UE.Source.OnPlay=null
   Sounds [
    &/IP_Jethro_CrustyChain_Overheat_Complete_01
    &/IP_Jethro_CrustyChain_Overheat_Complete_02
    &/IP_Jethro_CrustyChain_Overheat_Complete_03
   ]
  }
  nodes [
   n1 = Array.Shuffle:WaveAsset:Array(In Array=^Sounds, Next=^UE.Source.OnPlay)
   WavePlayer_0 = UE.Wave Player:Mono(Play=@n1.On Next, Wave Asset=@n1.Value)
   Multiply_0 = UE.Multiply:Audio by Float(AdditionalOperands=^MasterVolume, PrimaryOperand=@WavePlayer_0.Out Mono)
  ]
  links [{dst="@UE.OutputFormat.Mono.Audio:0", src=Multiply_0.Out}, {dst="@UE.Source.OneShot.OnFinished", src=WavePlayer_0.On Finished}]
 }]
properties {
 assetImportData=&self:AssetImportData
 attenuationSettings=&AS_Fort_WeaponFoley
 maxDistance=2250
 sampleRate=48000
 soundClassObject=&Fort_WeaponFoley
}
